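# Sound cue for the footsteps of the BangThePan emote in Athena.
# A Random node picks one of seven equally weighted footstep waves, and a Modulator varies its volume slightly while leaving pitch fixed.
# The cue plays at reduced volume with emote attenuation.
# Although it holds footsteps, it uses the emote music sound class and concurrency, and has a send to the music audio-analysis submix.
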SoundCue /Game/Athena/Sounds/Emotes/BangThePan/Emote_BangThePan_Footsteps
root=Modulator_0
nodes [
 {id=Modulator_0, children [Random_1], properties {pitchMax=1, pitchMin=1, volumeMax=1, volumeMin=0.8}}
 {
  id=Random_1
  children [WavePlayer_5, WavePlayer_6, WavePlayer_7, WavePlayer_8, WavePlayer_9, WavePlayer_10, WavePlayer_11]
  properties {bShouldExcludeFromBranchCulling=true, weights [1,1,1,1,1,1,1]}
 }
 {id=WavePlayer_10, properties {soundWaveAssetPtr=&/Emote_BangThePan_Footsteps_05}}
 {id=WavePlayer_11, properties {soundWaveAssetPtr=&/Emote_BangThePan_Footsteps_06}}
 {id=WavePlayer_5, properties {soundWaveAssetPtr=&/Emote_BangThePan_Footsteps_07}}
 {id=WavePlayer_6, properties {soundWaveAssetPtr=&/Emote_BangThePan_Footsteps_01}}
 {id=WavePlayer_7, properties {soundWaveAssetPtr=&/Emote_BangThePan_Footsteps_02}}
 {id=WavePlayer_8, properties {soundWaveAssetPtr=&/Emote_BangThePan_Footsteps_03}}
 {id=WavePlayer_9, properties {soundWaveAssetPtr=&/Emote_BangThePan_Footsteps_04}}
]
properties {
 attenuationSettings=&Fort_Emotes_Athena_Att
 maxDistance=1600
 soundClassObject=&Fort_Music_Emote
 volumeMultiplier=0.3
 concurrencySet [/Game/Sounds/Concurrency_Rules/Athena_EmoteMusic_Conc]
 soundSubmixSends [{minSendLevel=0, soundSubmix=&Fortnite_AudioAnalysis_Music}]
}
