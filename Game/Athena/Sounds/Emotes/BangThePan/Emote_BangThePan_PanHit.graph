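# Sound cue for the pan hit in the Bang The Pan emote.
# A random node picks among six pan-hit waves and one rarer, half-weight "annoying impact" wave boosted in volume by a modulator.
# It uses emote attenuation, the emote music sound class and concurrency, and has a send to the music audio-analysis submix.
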SoundCue /Game/Athena/Sounds/Emotes/BangThePan/Emote_BangThePan_PanHit
root=Random_0
nodes [
 {id=Modulator_0, children [WavePlayer_6], properties {volumeMax=1.7, volumeMin=1.7}}
 {
  id=Random_0
  children [WavePlayer_0, WavePlayer_1, WavePlayer_2, WavePlayer_3, WavePlayer_4, WavePlayer_12, Modulator_0]
  properties {bShouldExcludeFromBranchCulling=true, weights [1,1,1,1,1,1,0.5]}
 }
 {id=WavePlayer_0, properties {soundWaveAssetPtr=&/Emote_BangThePan_PanHits_01}}
 {id=WavePlayer_1, properties {soundWaveAssetPtr=&/Emote_BangThePan_PanHits_02}}
 {id=WavePlayer_12, properties {soundWaveAssetPtr=&/Emote_BangThePan_PanHits_06}}
 {id=WavePlayer_2, properties {soundWaveAssetPtr=&/Emote_BangThePan_PanHits_03}}
 {id=WavePlayer_3, properties {soundWaveAssetPtr=&/Emote_BangThePan_PanHits_04}}
 {id=WavePlayer_4, properties {soundWaveAssetPtr=&/Emote_BangThePan_PanHits_05}}
 {id=WavePlayer_6, properties {soundWaveAssetPtr=&/Emote_BangThePan_AnnoyingImpact_01}}
]
properties {
 attenuationSettings=&Fort_Emotes_Athena_Att
 maxDistance=1600
 soundClassObject=&Fort_Music_Emote
 volumeMultiplier=0.3
 concurrencySet [/Game/Sounds/Concurrency_Rules/Athena_EmoteMusic_Conc]
 soundSubmixSends [{minSendLevel=0, soundSubmix=&Fortnite_AudioAnalysis_Music}]
}
